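# Unlit translucent material for ash in the Cozy front-end environment, with no animation in the graph.
# A cloud falloff mask blends two color parameters, and the result is multiplied by two samples of one tiling noise texture at separate tilings.
# The opacity branch multiplies the first noise by the same mask and a gain, saturates it, softens intersections with DepthFade, then applies a final opacity scale.
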Material /Game/Environments/FrontEnd/Materials/Cozy/M_CozyAsh
graph {
 properties {blendMode=BLEND_Translucent, pixelDepthOffsetMode=PDOM_Legacy, shadingModel=MSM_Unlit, usedShadingModels=Unlit}
 nodes [
  n1 = TextureSample() {samplerType=SAMPLERTYPE_Masks, texture=/Game/Environments/World/Backgrounds/Locales/Temperate/Textures/T_CloudFalloff_M}
  n2 = Multiply(A=TextureCoordinate(), B=ScalarParameter() {defaultValue=4.3537, parameterName=Tiling2})
  n3 = TextureSample(UVs=Multiply(A=TextureCoordinate(), B=ScalarParameter() {defaultValue=1, parameterName=Tiling1})) {texture=/Game/Global/Textures/Default/Noise/T_Tiling_Noise_01}
  Multiply(
   A=LinearInterpolate(
    A=VectorParameter() {parameterName=Color, defaultValue=<rgba 0 0 0 1>}.RGB
    B=VectorParameter() {parameterName=Color2, defaultValue=<rgba 2 0.7294 0.2658 1>}.RGB
    Alpha=@n1.R
   )
   B=Multiply(A=TextureSample(UVs=@n2) {texture=&T_Tiling_Noise_01}.R, B=@n3.R)
  )
  n4 = Multiply(A=Multiply(A=@n3.RGB, B=@n1.R), B=ScalarParameter() {defaultValue=20.7953, parameterName=Opacity})
  Multiply(
   A=DepthFade(Opacity=Saturate(None=@n4), FadeDistance=ScalarParameter() {defaultValue=2.837, parameterName=Depth})
   B=ScalarParameter() {defaultValue=0.8774, parameterName=Opacity2}
  )
 ]
}
