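# Additive unlit material for the Cozy front-end environment that fakes a flickering colored light glow on a depth-test-free camera-facing sprite.
# Vertex offset pulls the sprite toward the camera and scales it by near fade.
# Two-lobe falloff compares view directions to the light and scene behind it, with noise flicker and MPC_Cozy low-detail variants.
# Opacity uses depth and distance fades, plus scene world-normal lighting on non-mobile shading paths.
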Material /Game/Environments/FrontEnd/Materials/Cozy/M_CozyLight
graph {
 properties {bDisableDepthTest=true, blendMode=BLEND_Additive, pixelDepthOffsetMode=PDOM_Legacy, shadingModel=MSM_Unlit, usedShadingModels=Unlit}
 nodes [
  n1 = MaterialFunctionCall() {materialFunction=&CameraDirectionVector}
  n2 = ScalarParameter() {defaultValue=30, parameterName=NearClip2}
  n3 = CameraPositionWS()
  n4 = ObjectPositionWS()
  n5 = Normalize(VectorInput=Subtract(A=@n3, B=@n4))
  n6 = Subtract(A=Distance(A=@n4, B=@n3), B=ScalarParameter() {defaultValue=30, parameterName=NearClip})
  n7 = ObjectPositionWS()
  n8 = LinearInterpolate(
   A=ScalarParameter() {defaultValue=256.2612, parameterName=Scale2}
   B=ScalarParameter() {defaultValue=71.5893, parameterName=Scale}
   Alpha=MaterialFunctionCall(Distance=ScalarParameter() {defaultValue=1000, parameterName=NearFade}) {materialFunction=&Near_Camera_Fade}.Result
  )
  Add(
   A=Multiply(A=@n1.Result, B=@n2)
   B=Add(
    A=Multiply(A=@n5, B=@n6)
    B=MaterialFunctionCall!(Center Location=@n7, 0-1 UVs, XY scale=@n8, OverridePivot, "Normals (Optional)") {materialFunction=&Sprite}.World position offset
   )
  )
  CameraVectorWS()
  Constant() {r=1}
  DepthFade()
  n9 = TextureCoordinate()
  n10 = Constant2Vector() {g=0.5, r=0.5}
  n11 = ScalarParameter() {defaultValue=0.2286, parameterName=LightRadius}
  n12 = Subtract(A=Distance(A=@n9, B=@n10), B=@n11)
  Divide(A=Constant() {r=1}, B=Multiply(A=@n12, B=@n12))
  LinearInterpolate(B=@n2, Alpha=Saturate(None=DotProduct(A=@n1.Result, B=@n5))) {constB=0}
  LinearInterpolate(
   A=Subtract(A=ObjectPositionWS(), B=WorldPosition().XYZ)
   B=MaterialFunctionCall!(Object Basis 1, Object Basis 2, Object Basis 3, Camera Space Position Offset=Constant3Vector() {constant=<rgba 0 30 0 0>}) {materialFunction=/Engine/Functions/Engine_MaterialFunctions02/WorldPositionOffset/AttachMeshToTheCamera}.World Position Offset
  )
  MaterialFunctionCall() {materialFunction=&CameraDirectionVector}
  MaterialFunctionCall() {materialFunction=&WorldPositionBehindTranslucency}
  n13 = Multiply(A=Time(), B=ScalarParameter() {defaultValue=0.3714, parameterName=Filcker})
  n14 = CameraPositionWS()
  n15 = MaterialFunctionCall() {materialFunction=&WorldPositionBehindTranslucency}
  n16 = ObjectPositionWS()
  n17 = DotProduct(A=Normalize(VectorInput=Subtract(A=@n14, B=@n15.Result)), B=Normalize(VectorInput=Subtract(A=@n14, B=@n16)))
  n18 = ScalarParameter() {defaultValue=30, parameterName=Falloff}
  n19 = ScalarParameter() {defaultValue=30, parameterName=FalloffLowDetai}
  n20 = CollectionParameter() {collection=&/MPC_Cozy, parameterName=isLowDetail}
  n21 = Subtract(A=@n17, B=ScalarParameter() {defaultValue=0.7, parameterName=FalloffOffset})
  n22 = LinearInterpolate(
   A=ScalarParameter() {defaultValue=0.2904, parameterName=Falloff2}
   B=ScalarParameter() {defaultValue=0.4907, parameterName=Falloff2LowDetail}
   Alpha=@n20
  )
  n23 = Add(
   A=Power(Base=@n17, Exp=LinearInterpolate(A=@n18, B=@n19, Alpha=@n20))
   B=Multiply(A=Power(Base=@n21, Exp=@n22), B=ScalarParameter() {defaultValue=0.2, parameterName=Falloff2Brightness})
  )
  n24 = Multiply(
   A=@n23
   B=LinearInterpolate(
    A=VectorParameter() {parameterName=Color, defaultValue=<rgba 0.9271 0.2454 0.1252 1>}.RGB
    B=VectorParameter() {parameterName=ColorLowDetail, defaultValue=<rgba 0.9271 0.2454 0.1252 1>}.RGB
    Alpha=@n20
   )
  )
  n25 = Multiply(A=ScalarParameter() {defaultValue=0.4407, parameterName=BrightnessPC}, B=@n24)
  n26 = Multiply(A=@n24, B=ScalarParameter() {defaultValue=0.1645, parameterName=BrightnessMobile})
  Multiply(
   A=LinearInterpolate(
    A=ScalarParameter() {defaultValue=0.4571, parameterName=Min}
    B=ScalarParameter() {defaultValue=1, parameterName=Max}
    Alpha=TextureSample(UVs=@n13) {texture=/Game/Global/Textures/Default/Noise/T_Tiling_Noise_01}.R
   )
   B=ShadingPathSwitch(Default=@n25, Deferred=@n25, Forward=@n26, Mobile=@n26)
  )
  n27 = DotProduct(
   A=Normalize(VectorInput=Subtract(A=@n16, B=@n15.Result))
   B=ComponentMask(None=SceneTexture() {sceneTextureId=PPI_WorldNormal}.Color) {b=true, g=true, r=true}
  )
  n28 = DepthFade(Opacity=@n23, FadeDistance=ScalarParameter() {defaultValue=1154.1626, parameterName=Depth2})
  n29 = Multiply(A=Max(A=Saturate(None=@n27), B=ScalarParameter() {defaultValue=0.3143, parameterName=MinLight}), B=@n28)
  n30 = Multiply(A=@n28, B=ScalarParameter() {defaultValue=1, parameterName=OpacityMobile})
  n31 = ObjectPositionWS()
  n32 = CameraPositionWS()
  n33 = Divide(
   A=Subtract(A=Distance(A=@n31, B=@n32), B=ScalarParameter() {defaultValue=1500, parameterName=FadeDistanceOffset})
   B=ScalarParameter() {defaultValue=1500, parameterName=FadeDistanceLength}
  )
  Multiply(A=ShadingPathSwitch(Default=@n29, Deferred=@n29, Forward=@n30, Mobile=@n30), B=OneMinus(None=Saturate(None=@n33)))
  Multiply(A=@n11, B=@n23)
  Normalize(VectorInput=Transform(None=Constant3Vector() {constant=<rgba 1 0 0 0>}) {transformSourceType=TRANSFORMSOURCE_Camera})
  ScalarParameter() {parameterName=Depth}
  SphereMask(A=@n9, B=@n10, Radius=Constant() {r=0.5}, Hardness=Constant())
  TextureCoordinate()
  VertexNormalWS()
 ]
}
