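# Master material for wrapped presents in the CozyLodge front end, with selectable paper patterns, colors, ribbon and gift tag.
# Mask channels blend paper, ribbon and tag attributes, with flake and optional damage normals layered on top.
# Static switches add fake shadow, tint, constellation and view-dependent glow, or bypass animation.
# World Position Offset collapses vertices to the object pivot, selecting ribbon types by vertex alpha and driving whole-mesh fadeout.
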Material /Game/Environments/FrontEnd/Materials/Cozy/M_CozyLodge_Anim
graph {
 properties {bUseMaterialAttributes=true, bUsedWithNanite=true, pixelDepthOffsetMode=PDOM_Legacy, usedShadingModels=Default Lit}
 nodes [
  Constant() {r=1}
  n1 = ScalarParameter() {parameterName=PaperType}
  n2 = VectorParameter() {parameterName=PaperColor1, defaultValue=<rgba 0.6823 0.0197 0 1>}
  n3 = If(
   A=@n1
   B=Constant() {r=0.5}
   "A > B"=VectorParameter() {parameterName=PaperColor2, defaultValue=<rgba 0 0.0656 0.4792 1>}.RGB
   "A == B"=@n2.RGB
   "A < B"=@n2.RGB
  )
  n4 = If(
   A=@n1
   B=Constant() {r=1.5}
   "A > B"=VectorParameter() {parameterName=PaperColor3, defaultValue=<rgba 0 0.5156 0.178 1>}.RGB
   "A == B"=@n3
   "A < B"=@n3
  )
  n5 = If(
   A=@n1
   B=Constant() {r=2.5}
   "A > B"=VectorParameter() {parameterName=PaperColor4, defaultValue=<rgba 0.4484 0.526 0.0719 1>}.RGB
   "A == B"=@n4
   "A < B"=@n4
  )
  n6 = If(
   A=@n1
   B=Constant() {r=2.5}
   "A > B"=VectorParameter() {parameterName=PaperColor5, defaultValue=<rgba 0 0 0 1>}.RGB
   "A == B"=@n5
   "A < B"=@n5
  )
  If(
   A=@n1
   B=Constant() {r=2.5}
   "A > B"=VectorParameter() {parameterName=PaperColor6, defaultValue=<rgba 0.3854 0.0767 0.1825 1>}.RGB
   "A == B"=@n6
   "A < B"=@n6
  )
  n7 = Multiply(A=TextureCoordinate(), B=ScalarParameter() {defaultValue=8, parameterName=Tiling})
  n8 = TextureSampleParameter2D(UVs=Rotator(Coordinate=@n7, Time=ScalarParameter() {defaultValue=1.9, parameterName=PatternRotation, sliderMax=4}) {speed=1}) {parameterName=PatternMasks, samplerType=SAMPLERTYPE_LinearColor, texture=/Game/Athena/UI/Frontend/WinterQuest/WF_Art/T_UI_DF_WrappingPapers}
  Multiply(A=@n8.RGB)
  n9 = TextureSampleParameter2D() {parameterName=Normal, samplerType=SAMPLERTYPE_Normal, texture=/Game/Gadgets/Assets/High_Rez_Present/Textures/T_High_Rez_Present_N}
  n10 = VectorParameter() {group=OverlayGradient, parameterName=OverlayGradientInner_Red, defaultValue=<rgba 0.8333 0.0391 0.026 1>}
  n11 = VectorParameter() {parameterName=ColorPaper1, defaultValue=<rgba 1 0 0 1>}
  n12 = LinearInterpolate(B=@n10.RGB, Alpha=ComponentMask(None=@n11.RGB) {r=true})
  n13 = VectorParameter() {group=OverlayGradient, parameterName=OverlayGradientInner_Green, defaultValue=<rgba 0.0738 0.625 0.0481 1>}
  n14 = LinearInterpolate(B=@n13.RGB, Alpha=ComponentMask(None=@n11.RGB) {g=true})
  n15 = VectorParameter() {group=OverlayGradient, parameterName=OverlayGradientInner_Blue, defaultValue=<rgba 0.0537 0.5693 0.9375 1>}
  n16 = ComponentMask(None=@n11.RGB) {b=true}
  n17 = VectorParameter() {group=OverlayGradient, parameterName=OverlayGradientInner_Silver, defaultValue=<rgba 0.4911 0.536 0.7604 1>}
  n18 = ComponentMask(None=@n11.RGB) {g=true}
  n19 = ComponentMask(None=@n11.RGB) {b=true}
  n20 = LinearInterpolate(B=@n17.RGB, Alpha=Saturate(None=Multiply(A=@n18, B=@n19) {constB=0}))
  n21 = VectorParameter() {group=OverlayGradient, parameterName=OverlayGradientInner_Purple, defaultValue=<rgba 0.4333 0.0394 0.6302 1>}
  n22 = ComponentMask(None=@n11.RGB) {r=true}
  n23 = ComponentMask(None=@n11.RGB) {b=true}
  n24 = LinearInterpolate(B=@n21.RGB, Alpha=Saturate(None=Multiply(A=@n22, B=@n23) {constB=0}))
  n25 = VectorParameter() {group=OverlayGradient, parameterName=OverlayGradientInner_Gold, defaultValue=<rgba 0.9375 0.5023 0.0049 1>}
  n26 = Saturate(None=Multiply(A=ComponentMask(None=@n11.RGB) {r=true}, B=ComponentMask(None=@n11.RGB) {g=true}) {constB=0})
  n27 = Add(A=Multiply(A=ComponentMask(None=@n11.RGB) {r=true}) {constB=0.5}, B=Multiply(A=ComponentMask(None=@n11.RGB) {g=true}) {constB=0.5})
  n28 = Multiply(A=ComponentMask(None=@n11.RGB) {b=true}) {constB=0.5}
  n29 = LinearInterpolate(
   A=Add(A=Add(A=@n12, B=@n14), B=LinearInterpolate(B=@n15.RGB, Alpha=@n16))
   B=Add(A=Add(A=@n20, B=@n24), B=LinearInterpolate(B=@n25.RGB, Alpha=@n26))
   Alpha=Floor(None=Add(A=@n27, B=@n28))
  )
  n30 = Desaturation(None=@n29, Fraction=ScalarParameter() {defaultValue=0.5, parameterName=DesatPaper2})
  n31 = ScalarParameter() {defaultValue=2, parameterName=BrightnessPaper2}
  n32 = Round(None=DotProduct(A=@n8.RGB, B=VectorParameter() {parameterName=PaperType, defaultValue=<rgba 0 0 1 1>}.RGB))
  n33 = TextureSampleParameter2D() {parameterName=Spec, samplerType=SAMPLERTYPE_Masks, texture=/Game/Gadgets/Assets/High_Rez_Present/Textures/T_High_Rez_Present_S}
  n34 = SetMaterialAttributes!(
   MaterialAttributes
   Base Color=LinearInterpolate(A=Multiply(A=@n30, B=@n31), B=@n29, Alpha=OneMinus(None=@n32))
   Metallic=@n33.G
   Specular=@n33.G
   Roughness=Multiply(A=ScalarParameter() {defaultValue=1, parameterName=RoughnessPaper}, B=@n33.B)
  ) {attributeSetTypes [
    #1
    #2
    #3
    #4
   ]}
  n35 = SetMaterialAttributes!(
   MaterialAttributes
   Base Color=VectorParameter() {parameterName=ColorRibbon, defaultValue=<rgba 0 0.3932 1 1>}.RGB
   Metallic=ScalarParameter() {defaultValue=0.424, parameterName=MetallicRibbon, sliderMax=1}
   Roughness=Multiply(A=ScalarParameter() {defaultValue=0.9639, parameterName=RoughnessRibbon, sliderMax=1}, B=@n33.B)
   Specular
  ) {attributeSetTypes [
    #1
    #2
    #4
    #3
   ]}
  n36 = TextureSampleParameter2D() {parameterName=Mask, samplerType=SAMPLERTYPE_Masks, texture=/Game/Gadgets/Assets/High_Rez_Present/Textures/T_High_Rez_Present_FX}
  n37 = TextureCoordinate()
  n38 = ComponentMask(None=VectorParameter() {parameterName=TagOffset, defaultValue=<rgba -0.414 -0.427 0 1>}.RGB) {g=true, r=true}
  n39 = ScalarParameter() {defaultValue=-0.035, parameterName=ScaleTagX}
  n40 = ScalarParameter() {defaultValue=-0.055, parameterName=ScaleTagY}
  n41 = Rotator(
   Coordinate=MaterialFunctionCall(UVs=Add(A=@n37, B=@n38), Texture Scale=AppendVector(A=@n39, B=@n40)) {materialFunction=&ScaleUVsByCenter}.UVs
   Time=ScalarParameter() {parameterName=TagRotation}
  )
  n42 = TextureSample!(UVs=@n41, Tex, Bias, Apply View MipBias) {mipValueMode=TMVM_MipBias, texture=/Game/Environments/FrontEnd/Textures/Cozy/CozyTag6}
  n43 = VectorParameter() {parameterName=Tag, defaultValue=<rgba 0 0 1 1>}
  n44 = TextureSample(UVs=@n41) {mipValueMode=TMVM_MipBias, texture=/Game/Environments/FrontEnd/Textures/Cozy/CozyTag5}
  n45 = OneMinus(None=@n43.RGB)
  n46 = ComponentMask(None=@n43.RGB) {r=true}
  n47 = ComponentMask(None=@n43.RGB) {g=true}
  n48 = Add(A=Add(A=Multiply(A=@n46) {constB=0.5}, B=Multiply(A=@n47) {constB=0.5}), B=Multiply(A=ComponentMask(None=@n43.RGB) {b=true}) {constB=0.5})
  n49 = Multiply(
   A=VectorParameter() {parameterName=ColorTag, defaultValue=<rgba 0.8385 0.4847 0.1308 1>}.RGB
   B=LinearInterpolate(A=DotProduct(A=@n42.RGB, B=@n43.RGB), B=DotProduct(A=@n44.RGB, B=@n45), Alpha=Floor(None=@n48))
  )
  n50 = Multiply(A=TextureCoordinate(), B=ScalarParameter() {defaultValue=21.8058, parameterName=Ribbon})
  n51 = LinearInterpolate(
   A=@n9.RGB
   B=MaterialFunctionCall(
    BaseNormal=@n9.RGB
    AdditionalNormal=TextureSample!(UVs=@n50, Tex, Apply View MipBias) {samplerType=SAMPLERTYPE_Normal, texture=/Game/Environments/Sets/Neo/Vehicles/Textures/T_Flakes_N}.RGB
   ) {materialFunction=&BlendAngleCorrectedNormals}.Result
   Alpha=Multiply(A=@n36.R, B=ScalarParameter() {defaultValue=0.08, parameterName=RibbonNormalIntensity, sliderMax=1})
  )
  n52 = TextureCoordinate()
  n53 = ScalarParameter() {defaultValue=1.3819, parameterName=DamageTiling}
  n54 = MaterialFunctionCall(
   Normal=TextureSample(UVs=Multiply(A=@n52, B=@n53)) {samplerType=SAMPLERTYPE_Normal, texture=/Game/Global/Textures/Default/Tileable_Patterns/CarDamageNormal}.RGB
   Flatness=ScalarParameter() {defaultValue=0.99, parameterName=Damage}
  ) {materialFunction=&FlattenNormal}
  n55 = SetMaterialAttributes!(
   MaterialAttributes=BlendMaterialAttributes(
    A=BlendMaterialAttributes(A=@n34, B=@n35, Alpha=@n36.R)
    B=SetMaterialAttributes!(MaterialAttributes, Base Color=@n49) {attributeSetTypes [#1]}
    Alpha=@n36.B
   )
   Normal=StaticSwitchParameter(
    True=MaterialFunctionCall(BaseNormal=@n51, AdditionalNormal=@n54.Result) {materialFunction=&BlendAngleCorrectedNormals}.Result
    False=@n51
   ) {defaultValue=true, parameterName=UseDamgePaper}
  ) {attributeSetTypes [#5]}
  n56 = Constant()
  n57 = Constant()
  n58 = VectorParameter() {parameterName=OffsetShadow, defaultValue=<rgba 0 0 0 1>}
  n59 = MaterialFunctionCall(
   UVs=Add(A=TextureCoordinate(), B=ComponentMask(None=@n58.RGB) {g=true, r=true})
   Texture Scale=ScalarParameter() {defaultValue=1, parameterName=ShadowScale}
  ) {materialFunction=&ScaleUVsByCenter}
  n60 = TextureSampleParameter2D(UVs=@n59.UVs) {
   parameterName=Shadow
   samplerType=SAMPLERTYPE_Masks
   texture=/Game/Environments/World/Backgrounds/Locales/Temperate/Textures/T_CloudFalloff_M
  }
  n61 = Multiply(A=ScalarParameter() {defaultValue=1, parameterName=ShadowIntensity}, B=@n60.R)
  n62 = StaticSwitchParameter(
   True=BlendMaterialAttributes(
    A=@n55
    B=SetMaterialAttributes!(MaterialAttributes, Base Color=@n56, Specular=@n57) {attributeSetTypes [#1, #3]}
    Alpha=Multiply(A=@n61, B=@n59.0-1 mask)
   )
   False=@n55
  ) {parameterName=UseFakeHelmetShadow}
  n63 = VertexColor()
  n64 = Constant() {r=1}
  n65 = Constant()
  n66 = If(A=@n63.A, B=Constant() {r=0.3}, "A > B"=@n64, "A == B"=@n64, "A < B"=@n65)
  n67 = If(A=@n63.A, B=Constant() {r=0.75}, "A > B"=@n65, "A == B"=@n65, "A < B"=@n64)
  n68 = Multiply(
   A=Subtract(A=ObjectPositionWS(), B=WorldPosition().XYZ)
   B=LinearInterpolate(A=Multiply(A=@n66, B=@n67), B=OneMinus(None=@n67), Alpha=ScalarParameter() {parameterName=RibbonType})
  )
  n69 = StaticSwitchParameter(
   True=@n62
   False=MaterialFunctionCall!(In=SetMaterialAttributes!(MaterialAttributes=@n62, World Position Offset=@n68) {attributeSetTypes [#6]}) {materialFunction=/Game/Environments/FrontEnd/MaterialFunctions/MF_CozyLodge_Anim}.Result
  ) {parameterName=IsStaticwithnoAnim}
  n70 = StaticSwitchParameter(
   True=MaterialFunctionCall!(
    Material=@n69
    Tint=TextureSampleParameter2D() {parameterName=AOFDiffuse, texture=/Game/Gadgets/Assets/Wrapped_Galileo_Rocket/Textures/Wrapped_Galileo_Rocket_D}.RGB
    TintMask
   ) {materialFunction=/Engine/Functions/MaterialLayerFunctions/MatLayerBlend_Tint}.Blended Material
   False=@n69
  ) {parameterName=IsAODIffuseTexture}
  n71 = StaticSwitchParameter(
   True=MaterialFunctionCall!(Mask=@n36.G, Normal=@n9.RGB, Material=@n70) {materialFunction=&/MF_Constellation_Present}.Result
   False=@n70
  ) {parameterName=useConstealation}
  n72 = VectorParameter() {group=Emissive, parameterName=EmissiveCenterPosition, defaultValue=<rgba 0 0 20 1>}
  n73 = CameraVectorWS()
  n74 = ScalarParameter() {defaultValue=-80, group=Emissive, parameterName=EmissiveCenterSoften}
  n75 = Subtract(
   A=WorldPosition() {worldPositionShaderOffset=WPT_ExcludeAllShaderOffsets}.XYZ
   B=Add(A=TransformPosition(None=@n72.RGB) {transformType=TRANSFORMPOSSOURCE_World}, B=Multiply(A=@n73, B=@n74))
  )
  n76 = Multiply(A=PixelNormalWS(), B=ScalarParameter() {defaultValue=5, parameterName=EmissivePixelNormalDistort})
  n77 = DotProduct(A=Normalize(VectorInput=Add(A=@n75, B=@n76)), B=@n73)
  n78 = Time()
  n79 = ScalarParameter() {defaultValue=0.1, parameterName=EmissiveDebugSpeed}
  n80 = StaticSwitchParameter(True=Frac(None=Multiply(A=@n78, B=@n79)), False=ScalarParameter() {parameterName=Progress_Buldge}) {parameterName=DebugEmissiveAnim}
  n81 = LinearInterpolate(
   A=ScalarParameter() {defaultValue=10, group=Emissive, parameterName=EmissiveFresnel_InterirorFalloff1}
   B=ScalarParameter() {defaultValue=5, group=Emissive, parameterName=EmissiveFresnel_InterirorFalloff2}
   Alpha=@n80
  )
  n82 = VectorParameter() {parameterName=EmissiveOpen, defaultValue=<rgba 30 25.3798 0 1>}
  n83 = Power(
   Base=TextureSample() {texture=&Wrapped_Galileo_Rocket_D}.R
   Exp=ScalarParameter() {defaultValue=3.0959, parameterName=EmissiveContrast}
  )
  n84 = ScalarParameter() {defaultValue=3, parameterName=EmissiveRampUpExponent}
  n85 = Multiply(A=Multiply(A=Power(Base=@n77, Exp=@n81), B=Multiply(A=@n82.RGB, B=@n83)), B=Saturate(None=Power(Base=@n80, Exp=@n84)))
  n86 = GetMaterialAttributes(None=StaticSwitchParameter(
    True=SetMaterialAttributes!(MaterialAttributes=@n71, Emissive Color=@n85) {attributeSetTypes [#7]}
    False=@n71
   ) {parameterName=UseLoPanEmissive}) {attributeGetTypes [#6]}
  n87 = Constant3Vector()
  n88 = Subtract(A=ObjectPositionWS(), B=Add(A=@n86.World Position Offset, B=WorldPosition() {worldPositionShaderOffset=WPT_ExcludeAllShaderOffsets}.XYZ))
  n89 = ScalarParameter() {parameterName=Fadeout}
  SetMaterialAttributes!(
   MaterialAttributes=@n86.MaterialAttributes
   World Position Offset=Add(A=@n86.World Position Offset, B=LinearInterpolate(A=@n87, B=@n88, Alpha=@n89))
  ) {attributeSetTypes [#6]}
  TextureSample(UVs=@n7) {samplerType=SAMPLERTYPE_Masks, texture=/Game/Weapons/WeaponSkins/Wraps/Textures/WrappingPaper/T_Wrap_WrappingPaper_M}
  VectorParameter() {parameterName=ColorPaper1, defaultValue=<rgba 1 0 0 1>}
  VectorParameter() {parameterName=ColorPaper2, defaultValue=<rgba 1 1 1 1>}
  WorldPosition()
 ]
}
