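# Default Lit rug material for the Cozy front-end environment, built from repurposed weapon-skin and shared noise textures.
# Two separately scaled, flattened normal maps, tile noise and fluffy cloth, are merged with angle-corrected blending.
# A wrapping-paper diffuse, tiled non-uniformly through a vector parameter, is partly desaturated and multiplied by a color tint.
# Scalar parameters provide roughness and specular values, while a wrapping-paper mask sample and an empty Multiply have no inputs wired.
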Material /Game/Environments/FrontEnd/Materials/Cozy/M_CozyRug
graph {
 properties {pixelDepthOffsetMode=PDOM_Legacy, usedShadingModels=Default Lit}
 nodes [
  n1 = Multiply(A=TextureCoordinate() {uTiling=0.22}, B=ScalarParameter() {defaultValue=0.8616, parameterName=Scale})
  n2 = Multiply(A=TextureCoordinate() {uTiling=0.22}, B=ScalarParameter() {defaultValue=26.2133, parameterName=Scale2})
  MaterialFunctionCall(
   BaseNormal=MaterialFunctionCall(
    Normal=TextureSample(UVs=@n1) {samplerType=SAMPLERTYPE_Normal, texture=/Game/Weapons/WeaponSkins/Textures/T_TileNoise_03_NRM}.RGB
    Flatness=ScalarParameter() {defaultValue=0.58, parameterName=NormalFlatten}
   ) {materialFunction=&FlattenNormal}.Result
   AdditionalNormal=MaterialFunctionCall(
    Normal=TextureSample(UVs=@n2) {samplerType=SAMPLERTYPE_Normal, texture=/Game/Global/Textures/Default/Noise/T_FN_FluffyCloth_N}.RGB
    Flatness=ScalarParameter() {defaultValue=0.4571, parameterName=NormalFlatten2}
   ) {materialFunction=&FlattenNormal}.Result
  ) {materialFunction=&BlendAngleCorrectedNormals}
  n3 = VectorParameter() {parameterName=ScaleDiffuse, defaultValue=<rgba 8.8449 2.2811 0 1>}
  n4 = Multiply(
   A=TextureCoordinate() {uTiling=0.22}
   B=Multiply(A=ScalarParameter() {defaultValue=1.4007, parameterName=Scale3}, B=ComponentMask(None=@n3.RGB) {g=true, r=true})
  )
  Multiply(
   A=VectorParameter() {parameterName=Color, defaultValue=<rgba 0.25 0.25 0.25 1>}.RGB
   B=Desaturation(
    None=TextureSample(UVs=@n4) {texture=/Game/Weapons/WeaponSkins/Wraps/Textures/WrappingPaper/T_Wrap_WrappingPaper_D}.RGB
    Fraction=ScalarParameter() {defaultValue=0.419, parameterName=desat}
   )
  )
  Multiply()
  ScalarParameter() {defaultValue=0.999, parameterName=Roughnes}
  ScalarParameter() {defaultValue=0.381, parameterName=Specular}
  TextureSample() {samplerType=SAMPLERTYPE_Masks, texture=/Game/Weapons/WeaponSkins/Wraps/Textures/WrappingPaper/T_Wrap_WrappingPaper_M}
 ]
}
