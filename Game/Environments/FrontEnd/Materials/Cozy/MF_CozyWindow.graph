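# Material function for the front-end cozy window that swings its left and right panels open as the camera nears an open position.
# Camera distance to that point, offset and normalized, drives a rotation angle that shrinks to zero farther away.
# Vertex color blue picks each side's pivot and swing direction, while doubled red scales the movement.
# It outputs the rotation as World Position Offset plus a corrected normal via FixRotateAboutAxisNormals.
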MaterialFunction /Game/Environments/FrontEnd/Materials/Cozy/MF_CozyWindow
graph {nodes [
  n1 = VertexColor()
  n2 = VectorParameter() {parameterName=Axis, defaultValue=<rgba 0 0 1 1>}
  n3 = Distance(
   A=TransformPosition(None=VectorParameter() {parameterName=OpenPosition, defaultValue=<rgba -132 0 192 1>}.RGB) {transformType=TRANSFORMPOSSOURCE_World}
   B=CameraPositionWS()
  )
  n4 = ScalarParameter() {defaultValue=-30, parameterName=Offset}
  n5 = Saturate(None=Divide(A=Add(A=@n3, B=@n4), B=ScalarParameter() {defaultValue=115.4878, parameterName=Distance}))
  n6 = Multiply(
   A=LinearInterpolate(A=Constant() {r=-1}, Alpha=@n1.B)
   B=Multiply(A=ScalarParameter() {defaultValue=0.2452, parameterName=Rotation}, B=OneMinus(None=@n5))
  )
  n7 = LinearInterpolate(
   A=TransformPosition(None=VectorParameter() {parameterName=PivotL, defaultValue=<rgba -198 4 0 1>}.RGB) {transformType=TRANSFORMPOSSOURCE_World}
   B=TransformPosition(None=VectorParameter() {parameterName=PivotR, defaultValue=<rgba -62 4 0 1>}.RGB) {transformType=TRANSFORMPOSSOURCE_World}
   Alpha=@n1.B
  )
  n8 = WorldPosition()
  FunctionOutput(None=Multiply(A=Add(A=@n1.R, B=@n1.R), B=RotateAboutAxis(NormalizedRotationAxis=@n2.RGB, RotationAngle=@n6, PivotPoint=@n7, Position=@n8.XYZ))) {outputName=WPO}
  FunctionOutput(None=MaterialFunctionCall(Rotation Axis=@n2.RGB, Rotation Angle=@n6) {materialFunction=&FixRotateAboutAxisNormals}.New Tangent Space Vertex Normal) {outputName=Normal}
 ]}
