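# Default Lit front-end material for a cozy cabin window, built on the red channel of a log cabin trim texture.
# That channel, remapped and clamped, blends two color parameters, brown by default, with a small share of the raw grayscale mixed back in.
# Two further parameterized lerps, R and S, remap the same channel, likely for roughness and specular.
# It also calls the MF_CozyWindow function, but no connections to it or to material outputs appear.
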Material /Game/Environments/FrontEnd/Materials/Cozy/M_CozyWindow
graph {
 properties {pixelDepthOffsetMode=PDOM_Legacy, usedShadingModels=Default Lit}
 nodes [
  n1 = TextureSample() {texture=/Game/Environments/Sets/Cabin01/Textures/T_FORT_Logcabin_Trim_D}
  LinearInterpolate(A=ScalarParameter() {defaultValue=0.5517, parameterName=R1}, B=ScalarParameter() {defaultValue=1.6046, parameterName=R2}, Alpha=@n1.R)
  LinearInterpolate(A=ScalarParameter() {parameterName=S1}, B=ScalarParameter() {defaultValue=0.0952, parameterName=S2}, Alpha=@n1.R)
  n2 = LinearInterpolate(A=ScalarParameter() {defaultValue=-2.0392, parameterName=1}, B=ScalarParameter() {defaultValue=29.974, parameterName=2}, Alpha=@n1.R)
  LinearInterpolate(
   A=LinearInterpolate(
    A=VectorParameter() {parameterName=Color2, defaultValue=<rgba 0.0521 0.0209 0.0061 1>}.RGB
    B=VectorParameter() {parameterName=Color, defaultValue=<rgba 0.125 0.0501 0.0146 1>}.RGB
    Alpha=Saturate(None=@n2)
   )
   B=@n1.R
   Alpha=ScalarParameter() {defaultValue=0.0857, parameterName=OgTexture}
  )
  MaterialFunctionCall() {materialFunction=&/MF_CozyWindow}
 ]
}
