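# Default Lit ground material for the Cozy front-end environment, with vertex displacement math, a flat color and layered normals.
# Pushes vertices horizontally away from the actor pivot using distance fade and inverted vertex green.
# Adds vertical X and Y sine waves masked by vertex green, sized by object scale.
# Blends two tiled swirl normals, flattening the second by vertex blue, and switches color for mobile.
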Material /Game/Environments/FrontEnd/Materials/Cozy/M_Cozy_Ground
graph {
 properties {pixelDepthOffsetMode=PDOM_Legacy, usedShadingModels=Default Lit}
 nodes [
  n1 = WorldPosition()
  n2 = VertexColor()
  n3 = Constant3Vector() {constant=<rgba 1 1 0 0>}
  n4 = Multiply(A=WorldPosition().XYZ, B=@n3)
  n5 = Multiply(A=@n3, B=ActorPositionWS())
  n6 = MaterialFunctionCall() {materialFunction=/Engine/Functions/Engine_MaterialFunctions02/WorldPositionOffset/ObjectScale}
  n7 = Add(A=Distance(A=@n4, B=@n5), B=Multiply(A=@n6.Scale X, B=ScalarParameter() {defaultValue=-400, parameterName=OffsetRadius}))
  n8 = Multiply(A=@n6.Scale X, B=ScalarParameter() {defaultValue=2220.8879, parameterName=RadiusFade})
  n9 = Add(A=OneMinus(None=@n2.G), B=Saturate(None=Divide(A=@n7, B=@n8)))
  n10 = Subtract(A=@n4, B=@n5)
  n11 = Constant3Vector()
  n12 = Multiply(
   A=Multiply(
    A=Saturate(None=@n9)
    B=MaterialFunctionCall(Vector=@n10, Default=@n11) {materialFunction=&SafeNormalize}.Result
   )
   B=ScalarParameter() {defaultValue=30248.4727, parameterName=ExpandPush}
  )
  n13 = Add(A=WorldPosition() {worldPositionShaderOffset=WPT_ExcludeAllShaderOffsets}.XYZ, B=@n12)
  n14 = ScalarParameter() {defaultValue=442.571, parameterName=Wavelength}
  n15 = Sine(None=Divide(A=ComponentMask(None=@n13) {r=true}, B=Multiply(A=@n14, B=@n6.Scale X)))
  n16 = Multiply(A=ScalarParameter() {defaultValue=9.5241, parameterName=Height}, B=@n6.Scale X)
  n17 = ScalarParameter() {defaultValue=417.7028, parameterName=Wavelength2}
  n18 = Sine(None=Divide(A=ComponentMask(None=@n13) {g=true}, B=Multiply(A=@n17, B=@n6.Scale X)))
  n19 = AppendVector(A=Constant2Vector(), B=Add(A=Multiply(A=@n15, B=@n16), B=Multiply(A=@n18, B=@n16)))
  Add(A=@n1.XYZ, B=Add(A=Multiply(A=@n2.G, B=@n19), B=@n12))
  Clamp(None=@n8) {clampMode=CMODE_ClampMin, minDefault=0.01}
  MaterialFunctionCall(
   TextureObject=TextureObject() {samplerType=SAMPLERTYPE_Normal, texture=/Game/Global/Textures/Default/Tileable_Patterns/T_Snow01_N}
   TextureSize=ScalarParameter() {defaultValue=6617.9727, parameterName=NormalSize}
   WorldSpace=StaticBool() {value=true}
   WorldPosition=@n1.XYZ
  ) {materialFunction=&WorldAlignedNormal}
  n20 = TextureCoordinate()
  n21 = ScalarParameter() {defaultValue=23.8525, parameterName=Tiling}
  n22 = Multiply(A=TextureCoordinate(), B=ScalarParameter() {defaultValue=4.8581, parameterName=Tiling2})
  MaterialFunctionCall(
   BaseNormal=TextureSample(UVs=Multiply(A=@n20, B=@n21)) {samplerType=SAMPLERTYPE_Normal, texture=/Game/Environments/Foliage/Textures/T_SwirlNoise_Soft_N}.RGB
   AdditionalNormal=LinearInterpolate(
    A=TextureSample(UVs=@n22) {samplerType=SAMPLERTYPE_Normal, texture=&T_SwirlNoise_Soft_N}.RGB
    B=Constant3Vector() {constant=<rgba 0 0 1 0>}
    Alpha=VertexColor().B
   )
  ) {materialFunction=&BlendAngleCorrectedNormals}
  ShadingPathSwitch(
   Default=VectorParameter() {parameterName=Color, defaultValue=<rgba 0.8646 0.8646 0.8646 1>}.RGB
   Mobile=VectorParameter() {parameterName=Mobile Color, defaultValue=<rgba 0.4635 0.4635 0.4635 1>}.RGB
  )
  TextureSample() {samplerType=SAMPLERTYPE_LinearGrayscale, texture=/Game/Global/Textures/Default/Noise/T_SwirlNoise_N_height}
 ]
}
