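# Surface material for the Cozy front-end environment that blends snow over a tinted rock texture, with Nanite usage enabled.
# The normal map is transformed and dotted with the up vector, then shaped by a contrast power to find upward-facing detail.
# That mask plus inverted red vertex color lerps tinted diffuse toward a snow color, with a separate snow color for mobile.
# Normal and spec maps are exposed as texture parameters.
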Material /Game/Environments/FrontEnd/Materials/Cozy/M_Cozy_SnowBlend
graph {
 properties {bUsedWithNanite=true, pixelDepthOffsetMode=PDOM_Legacy, usedShadingModels=Default Lit}
 nodes [
  n1 = OneMinus(None=VertexColor().R)
  n2 = TextureSampleParameter2D() {parameterName=Normal, samplerType=SAMPLERTYPE_Normal, texture=/Game/Environments/Apollo/Rocks/Natural_Rocks/Textures/T_NaturalRock_C_N}
  n3 = Transform(None=@n2.RGB)
  n4 = Constant3Vector() {constant=<rgba 0 0 1 0>}
  n5 = Saturate(None=Power(Base=DotProduct(A=@n3, B=@n4), Exp=ScalarParameter() {defaultValue=0.7088, parameterName=Contrast}))
  LinearInterpolate(
   A=Multiply(
    A=VectorParameter() {parameterName=Color, defaultValue=<rgba 1 1 1 1>}.RGB
    B=TextureSampleParameter2D() {parameterName=Diffuse, texture=/Game/Environments/Apollo/Rocks/Natural_Rocks/Textures/T_NaturalRock_C_D}.RGB
   )
   B=ShadingPathSwitch(
    Default=VectorParameter() {parameterName=Snow, defaultValue=<rgba 0.7917 0.7917 0.7917 1>}.RGB
    Mobile=VectorParameter() {parameterName=Snow Mobile, defaultValue=<rgba 0.5677 0.5677 0.5677 1>}.RGB
   )
   Alpha=Saturate(None=Add(A=@n1, B=@n5))
  )
  TextureSampleParameter2D() {parameterName=Spec, samplerType=SAMPLERTYPE_Masks, texture=/Game/Environments/Apollo/Rocks/Natural_Rocks/Textures/T_NaturalRock_C_S}
 ]
}
