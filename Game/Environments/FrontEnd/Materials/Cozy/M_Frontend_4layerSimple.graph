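# Four-layer Default Lit master material for Cozy front-end environment assets, usable on instanced static meshes and mesh particles.
# Each layer feeds tintable diffuse, masks, normal and emissive into MF_FN_TextureSetLerp_4Layers; layer 1 adds optional two-sided textures and neon-sign emissive.
# The result chains through decal, time-of-day emissive, fuzz, gradient, vertex-paint, up-vector, window-reflection and optional moss functions.
# Forward and mobile paths skip everything after the decal, with mobile keeping fewer attributes.
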Material /Game/Environments/FrontEnd/Materials/Cozy/M_Frontend_4layerSimple
graph {
 properties {
  bUseMaterialAttributes=true
  bUsedWithInstancedStaticMeshes=true
  bUsedWithMeshParticles=true
  ditheredLODTransition=true
  pixelDepthOffsetMode=PDOM_Legacy
  usedShadingModels=Default Lit
 }
 nodes [
  n1 = VectorParameter() {group=Material_3, parameterName=Emissive3, sortPriority=7, defaultValue=<rgba 0 0 0 1>}
  n2 = VectorParameter() {group=Material_4, parameterName=Emissive4, sortPriority=7, defaultValue=<rgba 0 0 0 1>}
  n3 = MaterialFunctionCall() {materialFunction=/Game/Athena/WinterArt/Materials/MF_UVsSet}
  n4 = TextureSampleParameter2D(UVs=@n3.UVs Layer 1) {
   group=Material_1
   parameterName=Diffuse Side 2
   samplerSource=SSM_Wrap_WorldGroupSettings
   sortPriority=1
   texture=/Game/Global/Textures/Default/Blanks/T_MidgreysRGB
  }
  n5 = TwoSidedSign()
  n6 = TextureSampleParameter2D(UVs=@n3.UVs Layer 1) {
   group=Material_1
   parameterName=Background Diffuse
   samplerSource=SSM_Wrap_WorldGroupSettings
   texture=&T_WhitesRGB
  }
  n7 = TextureSampleParameter2D(UVs=@n3.UVs Layer 1) {
   group=Material_1
   parameterName=Diffuse
   samplerSource=SSM_Wrap_WorldGroupSettings
   sortPriority=1
   texture=&T_MidgreysRGB
  }
  n8 = StaticSwitchParameter(True=@n6.A, False=@n7.A) {group=Material_1, parameterName=Use Color Swatch}
  n9 = TextureSampleParameter2D(UVs=@n3.UVs Layer 1) {
   group=Material_1
   parameterName=SpecularMasks
   samplerSource=SSM_Wrap_WorldGroupSettings
   samplerType=SAMPLERTYPE_Masks
   sortPriority=2
   texture=&T_Fortnite_Default_S
  }
  n10 = MaterialFunctionCall!(
   Transition=ScalarParameter() {defaultValue=0.5, group=Material_1, parameterName=L1_TintTransition, sortPriority=5}
   HM=@n8
   Contrast=ScalarParameter() {defaultValue=5, group=Material_1, parameterName=L1_TintContrast, sortPriority=5}
   Metalness=@n9.G
   BG Color=StaticSwitchParameter(True=@n6.RGB, False=@n7.RGB) {group=Material_1, parameterName=Use Color Swatch}
   TintColor=StaticSwitchParameter(True=@n7.RGB, False=VectorParameter() {group=Material_1, parameterName=TintColor, sortPriority=7, defaultValue=<rgba 1 1 1 0>}.RGB) {group=Material_1, parameterName=Use Color Swatch}
  ) {materialFunction=/Game/Athena/WinterArt/Materials/MF_ColorMask}
  n11 = AppendVector(
   A=StaticSwitchParameter(True=@n10.LerpMix, False=@n10.LerpMix) {group=Material_1, parameterName=L1_UseTintColorAsDiffuse}
   B=Multiply(A=@n8, B=ScalarParameter() {defaultValue=1, group=Material_1, parameterName=L1 Alpha, sortPriority=1})
  )
  n12 = MaterialFunctionCall() {materialFunction=&MF_UVsSet}
  n13 = TextureSampleParameter2D(UVs=@n12.UVs Layer 3) {
   desc=Use for trims and accents
   group=Material_3
   parameterName=Normals_Texture_3
   samplerSource=SSM_Wrap_WorldGroupSettings
   samplerType=SAMPLERTYPE_Normal
   sortPriority=3
   texture=/Engine/EngineMaterials/FlatNormal
  }
  n14 = TextureSampleParameter2D(UVs=@n12.UVs Layer 3) {
   group=Material_3
   parameterName=Background Diffuse 3
   samplerSource=SSM_Wrap_WorldGroupSettings
   texture=&T_WhitesRGB
  }
  n15 = TextureSampleParameter2D(UVs=@n12.UVs Layer 3) {
   desc=Use for trims and accents
   group=Material_3
   parameterName=Diffuse_Texture_3
   samplerSource=SSM_Wrap_WorldGroupSettings
   sortPriority=1
   texture=&T_MidgreysRGB
  }
  n16 = StaticSwitchParameter(True=@n14.A, False=@n15.A) {group=Material_3, parameterName=L3 Use Color Swatch}
  n17 = TextureSampleParameter2D(UVs=@n12.UVs Layer 3) {
   desc=Use for trims and accents
   group=Material_3
   parameterName=SpecularMasks_3
   samplerSource=SSM_Wrap_WorldGroupSettings
   samplerType=SAMPLERTYPE_Masks
   sortPriority=2
   texture=&T_Fortnite_Default_S
  }
  n18 = MaterialFunctionCall(
   Transition=ScalarParameter() {defaultValue=0.5, group=Material_3, parameterName=L3_TintTransition, sortPriority=5}
   HM=@n16
   Contrast=ScalarParameter() {defaultValue=1, group=Material_3, parameterName=L3_TintContrast, sortPriority=5}
   Metalness=@n17.G
   BG Color=StaticSwitchParameter(True=@n14.RGB, False=@n15.RGB) {group=Material_3, parameterName=L3 Use Color Swatch}
   TintColor=StaticSwitchParameter(True=@n15.RGB, False=VectorParameter() {group=Material_3, parameterName=M3_TintColor, sortPriority=5, defaultValue=<rgba 1 1 1 0>}.RGB) {group=Material_3, parameterName=L3 Use Color Swatch}
  ) {materialFunction=&MF_ColorMask}
  n19 = ScalarParameter() {defaultValue=1, group=Material_3, parameterName=L3 Alpha, sortPriority=1}
  n20 = MaterialFunctionCall() {materialFunction=&MF_UVsSet}
  n21 = TextureSampleParameter2D(UVs=@n20.UVs Layer 4) {
   group=Material_4
   parameterName=Background Diffuse 4
   samplerSource=SSM_Wrap_WorldGroupSettings
   texture=&T_WhitesRGB
  }
  n22 = TextureSampleParameter2D(UVs=@n20.UVs Layer 4) {
   desc=Use for doors
   group=Material_4
   parameterName=Diffuse_Texture_4
   samplerSource=SSM_Wrap_WorldGroupSettings
   sortPriority=1
   texture=&T_MidgreysRGB
  }
  n23 = StaticSwitchParameter(True=@n21.A, False=@n22.A) {group=Material_4, parameterName=L4 Use Color Swatch}
  n24 = TextureSampleParameter2D(UVs=@n20.UVs Layer 4) {
   desc=Use for doors
   group=Material_4
   parameterName=SpecularMasks_4
   samplerSource=SSM_Wrap_WorldGroupSettings
   samplerType=SAMPLERTYPE_Masks
   sortPriority=2
   texture=&T_Fortnite_Default_S
  }
  n25 = MaterialFunctionCall(
   Transition=ScalarParameter() {defaultValue=0.5, group=Material_4, parameterName=L4_TintTransition, sortPriority=5}
   HM=@n23
   Contrast=ScalarParameter() {defaultValue=1, group=Material_4, parameterName=L4_TintContrast, sortPriority=5}
   Metalness=@n24.G
   BG Color=StaticSwitchParameter(True=@n21.RGB, False=@n22.RGB) {group=Material_4, parameterName=L4 Use Color Swatch}
   TintColor=StaticSwitchParameter(True=@n22.RGB, False=VectorParameter() {group=Material_4, parameterName=M4_TintColor, sortPriority=5, defaultValue=<rgba 1 1 1 0>}.RGB) {group=Material_4, parameterName=L4 Use Color Swatch}
  ) {materialFunction=&MF_ColorMask}
  n26 = ScalarParameter() {defaultValue=1, group=Material_4, parameterName=L4 Alpha, sortPriority=1}
  n27 = TextureSampleParameter2D(UVs=@n20.UVs Layer 4) {
   desc=Use for doors
   group=Material_4
   parameterName=Normals_Texture_4
   samplerSource=SSM_Wrap_WorldGroupSettings
   samplerType=SAMPLERTYPE_Normal
   sortPriority=3
   texture=&FlatNormal
  }
  n28 = TextureObjectParameter() {group=Material_1, parameterName=Diffuse, sortPriority=1, texture=&T_MidgreysRGB}
  n29 = ScalarParameter() {group=Adv Emissive, parameterName=L1 X Speed, sortPriority=12}
  n30 = ScalarParameter() {group=Adv Emissive, parameterName=L1 Y Speed, sortPriority=13}
  n31 = VectorParameter() {group=Adv Emissive, parameterName=L1 Channel, sortPriority=9}
  n32 = ScalarParameter() {defaultValue=1, group=Adv Emissive, parameterName=L1 Offset, sliderMax=4, sliderMin=1, sortPriority=11}
  n33 = VectorParameter() {group=Material_1, parameterName=Emissive, sortPriority=7, defaultValue=<rgba 0 0 0 1>}
  n34 = Multiply(A=@n33.RGB, B=@n33.RGB)
  n35 = TextureSampleParameter2D(UVs=@n3.UVs Layer 1) {
   group=Material_1
   parameterName=SpecularMasks Side 2
   samplerSource=SSM_Wrap_WorldGroupSettings
   samplerType=SAMPLERTYPE_Masks
   sortPriority=2
   texture=&T_Fortnite_Default_S
  }
  n36 = VectorParameter() {group=Material_2, parameterName=Emissive2, sortPriority=7, defaultValue=<rgba 0 0 0 1>}
  n37 = TextureSampleParameter2D(UVs=@n3.UVs Layer 1) {
   group=Material_1
   parameterName=Normals Side 2
   samplerSource=SSM_Wrap_WorldGroupSettings
   samplerType=SAMPLERTYPE_Normal
   sortPriority=3
   texture=&FlatNormal
  }
  n38 = TextureSampleParameter2D(UVs=@n3.UVs Layer 1) {
   group=Material_1
   parameterName=Normals
   samplerSource=SSM_Wrap_WorldGroupSettings
   samplerType=SAMPLERTYPE_Normal
   sortPriority=3
   texture=&FlatNormal
  }
  n39 = MaterialFunctionCall() {materialFunction=&MF_UVsSet}
  n40 = TextureSampleParameter2D(UVs=@n39.UVs Layer 2) {
   group=Material_2
   parameterName=Background Diffuse 2
   samplerSource=SSM_Wrap_WorldGroupSettings
   texture=&T_WhitesRGB
  }
  n41 = TextureSampleParameter2D(UVs=@n39.UVs Layer 2) {
   desc=Use for interior walls and ceilings
   group=Material_2
   parameterName=Diffuse_Texture_2
   samplerSource=SSM_Wrap_WorldGroupSettings
   sortPriority=1
   texture=&T_MidgreysRGB
  }
  n42 = StaticSwitchParameter(True=@n40.A, False=@n41.A) {group=Material_2, parameterName=L2 Use Color Swatch}
  n43 = TextureSampleParameter2D(UVs=@n39.UVs Layer 2) {
   desc=Use for interior walls and ceilings
   group=Material_2
   parameterName=SpecularMasks_2
   samplerSource=SSM_Wrap_WorldGroupSettings
   samplerType=SAMPLERTYPE_Masks
   sortPriority=2
   texture=&T_Fortnite_Default_S
  }
  n44 = MaterialFunctionCall(
   Transition=ScalarParameter() {defaultValue=0.5, group=Material_2, parameterName=L2_TintTransition, sortPriority=5}
   HM=@n42
   Contrast=ScalarParameter() {defaultValue=1, group=Material_2, parameterName=L2_TintContrast, sortPriority=5}
   Metalness=@n43.G
   BG Color=StaticSwitchParameter(True=@n40.RGB, False=@n41.RGB) {group=Material_2, parameterName=L2 Use Color Swatch}
   TintColor=StaticSwitchParameter(True=@n41.RGB, False=VectorParameter() {group=Material_2, parameterName=M2_TintColor, sortPriority=5, defaultValue=<rgba 1 1 1 0>}.RGB) {group=Material_2, parameterName=L2 Use Color Swatch}
  ) {materialFunction=&MF_ColorMask}
  n45 = ScalarParameter() {defaultValue=1, group=Material_2, parameterName=L2 Alpha, sortPriority=1}
  n46 = TextureSampleParameter2D(UVs=@n39.UVs Layer 2) {
   desc=Use for interior walls and ceilings
   group=Material_2
   parameterName=Normals_Texture_2
   samplerSource=SSM_Wrap_WorldGroupSettings
   samplerType=SAMPLERTYPE_Normal
   sortPriority=3
   texture=&FlatNormal
  }
  n47 = MaterialFunctionCall!(
   Use3Materials
   Use4Materials
   Use2Materials
   D=StaticSwitchParameter(True=LinearInterpolate(A=@n4.RGBA, B=@n11, Alpha=@n5), False=@n11) {group=Material_1, parameterName=TwoSided, sortPriority=18}
   S=StaticSwitchParameter(True=LinearInterpolate(A=@n35.RGBA, B=@n9.RGBA, Alpha=@n5), False=@n9.RGBA) {group=Material_1, parameterName=TwoSided, sortPriority=18}
   N=StaticSwitchParameter(True=LinearInterpolate(A=@n37.RGB, B=@n38.RGB, Alpha=@n5), False=@n38.RGB) {group=Material_1, parameterName=TwoSided, sortPriority=18}
   D2=AppendVector(
    A=StaticSwitchParameter(True=@n44.LerpMix, False=@n44.LerpMix) {group=Material_2, parameterName=L2_UseTintColorAsDiffuse}
    B=Multiply(A=@n42, B=@n45)
   )
   S2=@n43.RGBA
   N2=@n46.RGB
   D3=AppendVector(
    A=StaticSwitchParameter(True=@n18.LerpMix, False=@n18.LerpMix) {group=Material_3, parameterName=L3_UseTintColorAsDiffuse}
    B=Multiply(A=@n16, B=@n19)
   )
   S3=@n17.RGBA
   N3=@n13.RGB
   D4=AppendVector(
    A=StaticSwitchParameter(True=@n25.LerpMix, False=@n25.LerpMix) {group=Material_4, parameterName=L4_UseTintColorAsDiffuse}
    B=Multiply(A=@n23, B=@n26)
   )
   S4=@n24.RGBA
   N4=@n27.RGB
   E1=StaticSwitchParameter(
    True=MaterialFunctionCall!(Base Diffuse=@n28) {materialFunction=/Game/Athena/WinterArt/Materials/MF_NeonSignEmissive}.Result
    False=MaterialFunctionCall!(X=@n29, Y=@n30, Dot=@n31.RGB, TextureOffset=@n32, EmissiveTexture=@n34) {materialFunction=/Game/Athena/WinterArt/Materials/MF_EmissiveMask}.EmissiveMasked
   ) {group=NeonEmissive, parameterName=UseNeonSignEmissive_Layer1}
   E2=MaterialFunctionCall(
    X=ScalarParameter() {group=Adv Emissive, parameterName=L2 X Speed, sortPriority=16}
    Y=ScalarParameter() {group=Adv Emissive, parameterName=L2 Y Speed, sortPriority=17}
    Dot=VectorParameter() {group=Adv Emissive, parameterName=L2 Channel, sortPriority=14}.RGB
    TextureOffset=ScalarParameter() {defaultValue=1, group=Adv Emissive, parameterName=L2 Offset, sliderMax=4, sliderMin=1, sortPriority=15}
    EmissiveTexture=Multiply(A=@n36.RGB, B=@n36.RGB)
   ) {materialFunction=&MF_EmissiveMask}.EmissiveMasked
   E3=MaterialFunctionCall(
    X=ScalarParameter() {group=Adv Emissive, parameterName=L3 X Speed, sortPriority=21}
    Y=ScalarParameter() {group=Adv Emissive, parameterName=L3 Y Speed, sortPriority=22}
    Dot=VectorParameter() {group=Adv Emissive, parameterName=L3 Channel, sortPriority=18}.RGB
    TextureOffset=ScalarParameter() {defaultValue=1, group=Adv Emissive, parameterName=L3 Offset, sliderMax=4, sliderMin=1, sortPriority=19}
    EmissiveTexture=Multiply(A=@n1.RGB, B=@n1.RGB)
   ) {materialFunction=&MF_EmissiveMask}.EmissiveMasked
   E4=MaterialFunctionCall(
    X=ScalarParameter() {group=Adv Emissive, parameterName=L4 X Speed, sortPriority=25}
    Y=ScalarParameter() {group=Adv Emissive, parameterName=L4 Y Speed, sortPriority=26}
    Dot=VectorParameter() {group=Adv Emissive, parameterName=L4 Channel, sortPriority=23}.RGB
    TextureOffset=ScalarParameter() {defaultValue=1, group=Adv Emissive, parameterName=L4 Offset, sliderMax=4, sliderMin=1, sortPriority=24}
    EmissiveTexture=Multiply(A=@n2.RGB, B=@n2.RGB)
   ) {materialFunction=&MF_EmissiveMask}.EmissiveMasked
   CCMaskByMetal_1
   CC1=ScalarParameter() {group=Material_1, parameterName=L1 Clearcoat}
   CCR_1
   CC2=ScalarParameter() {group=Material_2, parameterName=L2 Clearcoat}
   CCR_2
   CCMaskByMetal_2
   CC3=ScalarParameter() {group=Material_3, parameterName=L3 Clearcoat}
   CCR_3
   CCMaskByMetal_3
   CC4=ScalarParameter() {group=Material_4, parameterName=L4 Clearcoat}
   CCMaskByMetal_4
   CCR_4
   OpacityMaskCancel_1
   OpacityMaskCancel_2
   OpacityMaskCancel_3
   OpacityMaskCancel_4
   SSS_1
   SSS_2
   SSS_3
   SSS_4
   SSS_Depth_1
   SSS_Depth_2
   SSS_Depth_3
   SSS_Depth_4
  ) {materialFunction=/Game/Athena/WinterArt/Materials/MF_FN_TextureSetLerp_4Layers}
  n48 = ScalarParameter() {defaultValue=1, group=SSS, parameterName=SSSVal}
  n49 = SetMaterialAttributes!(
   MaterialAttributes
   Base Color=@n47.Diffuse
   Specular=@n47.Diffuse
   Metallic=@n47.Diffuse
   Roughness=@n47.Diffuse
   Emissive Color=@n47.Diffuse
   Normal=@n47.Diffuse
   Custom Data 0=Multiply(A=@n47.Clearcoat, B=@n47.Clearcoat)
   Custom Data 1=Constant() {r=0.025}
   Opacity Mask=@n47.Diffuse
   Subsurface Color=Multiply(A=@n47.Height, B=Multiply(A=@n47.Diffuse, B=@n48))
   Opacity=ScalarParameter() {defaultValue=0.5, group=SSS, parameterName=SSSDepth}
  ) {attributeSetTypes [
    #1
    #2
    #3
    #4
    #5
    #6
    #7
    #8
    #9
    #10
    #11
   ]}
  n50 = MaterialFunctionCall!(In=@n49) {materialFunction=/Game/Athena/WinterArt/Materials/MF_DecalPrint}
  n51 = MaterialFunctionCall!(In=MaterialFunctionCall!(In=MaterialFunctionCall!(In=@n50.Result) {materialFunction=/Game/Athena/WinterArt/Materials/MF_TOD_EmissiveColors}.Result) {materialFunction=/Game/Athena/WinterArt/Materials/MF_WS_Texture}.Result) {materialFunction=/Game/Athena/WinterArt/Materials/MF_Fuzzy}
  n52 = MaterialFunctionCall!(
   In=MaterialFunctionCall!(
    transitionFactor
    MaterialIn=MaterialFunctionCall!(transitionFactor, MaterialIn=@n51.Result) {materialFunction=/Game/Athena/WinterArt/Materials/MF_ObjectSpaceGradient}.Color*Gradient
   ) {materialFunction=/Game/Athena/WinterArt/Materials/MF_WSGradient}.Color*Gradient
   UseOneSampler
  ) {materialFunction=/Game/Athena/WinterArt/Materials/MF_VPTexture}
  n53 = MaterialFunctionCall!(In=MaterialFunctionCall!(In=@n52.Result, UpVectorTexture=@n52.Result) {materialFunction=/Game/Athena/WinterArt/Materials/MF_UpVectorTexture}.Result) {materialFunction=/Game/Athena/WinterArt/Materials/M_FakeWindowReflections}
  n54 = MaterialFunctionCall!(WorldSpaceNormals, In=@n53.Result) {materialFunction=/Game/Environments/Apollo/Rocks/Natural_Rocks/Materials/MossSoft2}
  n55 = GetMaterialAttributes(None=@n50.Result) {attributeGetTypes [
    #1
    #5
    #9
    #6
   ]}
  n56 = SetMaterialAttributes!(MaterialAttributes, Base Color=@n55.Base Color, Emissive Color=@n55.Base Color, Opacity Mask=@n55.Base Color, Normal=@n55.Base Color) {attributeSetTypes [
    #1
    #5
    #9
    #6
   ]}
  MaterialFunctionCall!(In=ShadingPathSwitch(Default=StaticSwitchParameter(True=@n54.Result, False=@n53.Result) {parameterName=UseMoss}, Forward=@n50.Result, Mobile=@n56)) {materialFunction=/Game/Athena/WinterArt/Materials/MF_GameWideShared}
 ]
}
