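# Lit material for blinking Christmas-style string lights on spline meshes in the cozy front-end environment.
# Time, offset by the fractional actor X position and scaled by Flickerspeed, is rounded into a hard on/off toggle that desynchronizes separate actors.
# The toggle swaps the emissive texture's alpha mask with its inverse, alternating lit regions, then multiplies the texture color and an HDR Emissive tint.
# Diffuse, normal and specular-mask textures are also sampled.
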Material /Game/Environments/FrontEnd/Materials/Cozy/M_StringLightSpline
graph {
 properties {bUsedWithSplineMeshes=true, pixelDepthOffsetMode=PDOM_Legacy, usedShadingModels=Default Lit}
 nodes [
  n1 = TextureSample() {samplerType=SAMPLERTYPE_Masks, texture=/Game/Environments/Props/Winter/Textures/T_StringLIghts_E}
  n2 = ActorPositionWS()
  n3 = Add(A=Time(), B=Frac(None=ComponentMask(None=@n2) {r=true}))
  n4 = ScalarParameter() {defaultValue=0.8283, parameterName=Flickerspeed}
  n5 = Round(None=Frac(None=Multiply(A=@n3, B=@n4)))
  Multiply(
   A=LinearInterpolate(A=@n1.A, B=OneMinus(None=@n1.A), Alpha=Saturate(None=@n5))
   B=Multiply(A=VectorParameter() {parameterName=Emissive, defaultValue=<rgba 10 10 10 1>}.RGB, B=@n1.RGB)
  )
  TextureSample() {samplerType=SAMPLERTYPE_Normal, texture=/Game/Environments/Props/Winter/Textures/T_StringLights_N}
  TextureSample() {texture=/Game/Environments/Props/Winter/Textures/T_StringLights_D}
  TextureSample() {samplerType=SAMPLERTYPE_Masks, texture=/Game/Environments/Props/Winter/Textures/T_StringLights_S}
  WorldPosition()
 ]
}
