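# Translucent unlit front-end material, named as an unlock attention cue, that builds a pulsing star-sparkle glint.
# A panned, world-aligned star texture plus a Fresnel rim is multiplied by a sine pulse with pauses.
# Object position offsets the pulse timing, so objects in different places flash out of sync.
# Screen/UV star blend, camera-offset and Color nodes sit apart from that chain, and no material output links are shown.
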Material /Game/Environments/FrontEnd/Materials/Cozy/UnlockAttention
graph {
 properties {
  bUsedWithMeshDeformer=true
  bUsedWithNanite=true
  bUsedWithSkeletalMesh=true
  blendMode=BLEND_Translucent
  pixelDepthOffsetMode=PDOM_Legacy
  shadingModel=MSM_Unlit
  usedShadingModels=Unlit
 }
 nodes [
  n1 = ScalarParameter() {defaultValue=613.5247, parameterName=Tiling}
  n2 = Multiply(A=ScreenPosition().ViewportUV, B=@n1)
  n3 = ScalarParameter() {defaultValue=0.0476, parameterName=Panspeed}
  n4 = Multiply(A=TextureCoordinate(), B=ScalarParameter() {defaultValue=0.1905, parameterName=Tiling2})
  n5 = ScalarParameter() {defaultValue=0.0286, parameterName=Panspeed2}
  n6 = Sine(None=Multiply(A=Time(), B=ScalarParameter() {defaultValue=0.2381, parameterName=SpeedSparkle}))
  LinearInterpolate(
   A=TextureSample(UVs=Panner(Coordinate=@n2, Speed=@n3)) {samplerType=SAMPLERTYPE_Masks, texture=/Game/Packages/Fortress_Sky/Textures/T_Sky_Stars_Sharp}.RGB
   B=TextureSample(UVs=Panner(Coordinate=@n4, Speed=@n5)) {samplerType=SAMPLERTYPE_Masks, texture=&T_Sky_Stars_Sharp}.RGB
   Alpha=Divide(A=Add(A=@n6))
  )
  n7 = CameraPositionWS()
  n8 = WorldPosition()
  Multiply(A=Normalize(VectorInput=Subtract(A=@n7, B=@n8.XYZ)), B=ScalarParameter() {defaultValue=5, parameterName=Offset})
  n9 = TextureObject() {samplerType=SAMPLERTYPE_Masks, texture=&T_Sky_Stars_Sharp}
  n10 = AppendVector(A=Constant(), B=ScalarParameter() {defaultValue=0.0476, parameterName=Panspeed})
  n11 = Fresnel(ExponentIn=ScalarParameter() {defaultValue=6, parameterName=fres}) {baseReflectFraction=0}
  n12 = ScalarParameter() {defaultValue=0.1, parameterName=Min}
  n13 = MaterialFunctionCall(Float3=ObjectPositionWS()) {materialFunction=&BreakOutFloat3Components}
  n14 = Add(A=Add(A=@n13.R, B=@n13.R), B=@n13.B)
  n15 = Add(A=Time(), B=Multiply(A=Frac(None=@n14), B=ScalarParameter() {defaultValue=33.45, parameterName=rand}))
  n16 = ScalarParameter() {defaultValue=0.0862, parameterName=Speedpause}
  n17 = Multiply(A=Frac(None=Multiply(A=@n15, B=@n16)), B=ScalarParameter() {defaultValue=3, parameterName=Pause})
  n18 = Divide(A=Frac(None=Saturate(None=@n17)))
  Multiply(
   A=Add(
    A=MaterialFunctionCall!(TextureObject=@n9, TextureSize=@n1, WorldPosition, Export Float 4, In=@n10, World Space Normal, ProjectionTransitionContrast) {materialFunction=/Game/Athena/SafeZone/WorldAlignedTexture_ScalePan}.XYZ Texture
    B=Multiply(A=@n11, B=@n12)
   )
   B=Saturate(None=Sine(None=@n18))
  )
  TextureCoordinate()
  VectorParameter() {parameterName=Color, defaultValue=<rgba 5 2.6485 0.8999 1>}
 ]
}
